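# Animation Blueprint for the Dumpling Man backpack cosmetic, adding secondary physics motion to its bottle and bun bones.
# The input pose runs through spring-constrained, damped Anim Dynamics on two bottle bones, then partially weighted additive translations on three bun bones.
# Anim Dynamics then simulates those buns with tight angular limits and lighter gravity.
# The event graph is empty, and the BunTranslation values it reads are not declared or set here.
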AnimBlueprint /BRCosmetics/Accessories/FORT_Backpacks/M_MED_Dumpling_Man_Backpack/Meshes/Backpack_M_MED_DumplingMan_AnimBP
target_skeleton=&/Backpack_M_MED_DumplingMan_Skeleton
graphs [
 AnimGraph {nodes [
   AnimGraphNode_SubInput_0 = Input Pose()
   n1 = Anim Dynamics - Bone: dyn_bottle_1(ComponentPose=Local To Component(LocalPose=@AnimGraphNode_SubInput_0), Alpha=1, ExternalForce="0,0,0") {
    angularDampingOverride=0.999
    angularSpringConstant=1000
    bAngularSpring=true
    bGravityOverrideInSimSpace=true
    bOverrideAngularDamping=true
    bOverrideLinearDamping=true
    bUseGravityOverride=true
    linearDampingOverride=0.999
    alphaBoolBlend {blendOption=Linear}
    boundBone {boneName=dyn_bottle_1}
    componentAppliedLinearAccClamp=<1000 1000 1000>
    componentLinearAccScale=<1 1 1>
    componentLinearVelScale=<1 1 1>
    gravityOverride {z=-980}
    physicsBodyDefinitions [{
      sphereCollisionRadius=0
      boundBone {boneName=dyn_bottle_1}
      boxExtents=<1 10 10>
      constraintSetup {angularTargetAxis=AxisY, angularLimitsMax=<0 3 8>, angularLimitsMin=<0 -3 -8>, angularTarget=<0 0 0>}
      localJointOffset=<5 0 0>
     }]
   }
   n2 = Anim Dynamics - Bone: dyn_bottle_2(ComponentPose=@n1, Alpha=1, ExternalForce="0,0,0") {
    angularDampingOverride=0.999
    angularSpringConstant=1000
    bAngularSpring=true
    bGravityOverrideInSimSpace=true
    bOverrideAngularDamping=true
    bOverrideLinearDamping=true
    bUseGravityOverride=true
    linearDampingOverride=0.999
    alphaBoolBlend {blendOption=Linear}
    boundBone {boneName=dyn_bottle_2}
    componentAppliedLinearAccClamp=<1000 1000 1000>
    componentLinearAccScale=<1 1 1>
    componentLinearVelScale=<1 1 1>
    gravityOverride {z=-980}
    physicsBodyDefinitions [{
      sphereCollisionRadius=0
      boundBone {boneName=dyn_bottle_2}
      boxExtents=<1 10 10>
      constraintSetup {angularTargetAxis=AxisY, angularLimitsMax=<0 2 8>, angularLimitsMin=<0 -2 -8>, angularTarget=<0 0 0>}
      localJointOffset=<5 0 0>
     }]
   }
   n3 = "Transform (Modify) Bone - Bone: dyn_bun_1"(ComponentPose=@n2, Alpha=0.3, Translation=Get Bun1Translation() -> (Bun1Translation:Vector), Rotation="0,0,0", Scale="1,1,1") {
    rotationSpace=BCS_BoneSpace
    translationMode=BMM_Additive
    translationSpace=BCS_BoneSpace
    alphaBoolBlend {blendOption=Linear}
    boneToModify {boneName=dyn_bun_1}
   }
   n4 = "Transform (Modify) Bone - Bone: dyn_bun_2"(ComponentPose=@n3, Alpha=0.3, Translation=Get Bun2Translation() -> (Bun2Translation:Vector), Rotation="0,0,0", Scale="1,1,1") {
    rotationSpace=BCS_BoneSpace
    translationMode=BMM_Additive
    translationSpace=BCS_BoneSpace
    alphaBoolBlend {blendOption=Linear}
    boneToModify {boneName=dyn_bun_2}
   }
   n5 = "Transform (Modify) Bone - Bone: dyn_bun_3"(ComponentPose=@n4, Alpha=0.3, Translation=Get Bun3Translation() -> (Bun3Translation:Vector), Rotation="0,0,0", Scale="1,1,1") {
    rotationSpace=BCS_BoneSpace
    translationMode=BMM_Additive
    translationSpace=BCS_BoneSpace
    alphaBoolBlend {blendOption=Linear}
    boneToModify {boneName=dyn_bun_3}
   }
   n6 = Anim Dynamics - Bone: dyn_bun_1(ComponentPose=@n5, Alpha=1, ExternalForce="0,0,0") {
    angularDampingOverride=0.999
    angularSpringConstant=2000
    bAngularSpring=true
    bGravityOverrideInSimSpace=true
    bOverrideAngularDamping=true
    bOverrideLinearDamping=true
    bUseGravityOverride=true
    gravityScale=0.5
    linearDampingOverride=0.999
    alphaBoolBlend {blendOption=Linear}
    boundBone {boneName=dyn_bun_1}
    componentAppliedLinearAccClamp=<1000 1000 1000>
    componentLinearAccScale=<1 1 1>
    componentLinearVelScale=<1 1 1>
    gravityOverride {z=-600}
    physicsBodyDefinitions [{
      sphereCollisionRadius=0
      boundBone {boneName=dyn_bun_1}
      boxExtents=<1 5 5>
      constraintSetup {
       angularTargetAxis=AxisY
       twistAxis=AxisY
       angularLimitsMax=<0.2 1 1>
       angularLimitsMin=<-0.2 -1 -1>
       angularTarget=<0 0 0>
       linearAxesMax=<1 0.1 0>
       linearAxesMin=<0 -0.1 0>
      }
      localJointOffset=<2 0 0>
     }]
   }
   n7 = Anim Dynamics - Bone: dyn_bun_2(ComponentPose=@n6, Alpha=1, ExternalForce="0,0,0") {
    angularDampingOverride=0.999
    angularSpringConstant=2000
    bAngularSpring=true
    bGravityOverrideInSimSpace=true
    bOverrideAngularDamping=true
    bOverrideLinearDamping=true
    bUseGravityOverride=true
    gravityScale=0.5
    linearDampingOverride=0.999
    alphaBoolBlend {blendOption=Linear}
    boundBone {boneName=dyn_bun_2}
    componentAppliedLinearAccClamp=<1000 1000 1000>
    componentLinearAccScale=<1 1 1>
    componentLinearVelScale=<1 1 1>
    gravityOverride {z=-650}
    physicsBodyDefinitions [{
      sphereCollisionRadius=0
      boundBone {boneName=dyn_bun_2}
      boxExtents=<1 5 5>
      constraintSetup {angularTargetAxis=AxisY, angularLimitsMax=<0.1 1 1>, angularLimitsMin=<-0.1 -1 -1>, angularTarget=<0 0 0>, linearAxesMax=<1 0 0>}
      localJointOffset=<2 0 0>
     }]
   }
   n8 = Anim Dynamics - Bone: dyn_bun_3(ComponentPose=@n7, Alpha=1, ExternalForce="0,0,0") {
    angularDampingOverride=0.999
    angularSpringConstant=2000
    bAngularSpring=true
    bGravityOverrideInSimSpace=true
    bOverrideAngularDamping=true
    bOverrideLinearDamping=true
    bUseGravityOverride=true
    gravityScale=0.5
    linearDampingOverride=0.999
    alphaBoolBlend {blendOption=Linear}
    boundBone {boneName=dyn_bun_3}
    componentAppliedLinearAccClamp=<1000 1000 1000>
    componentLinearAccScale=<1 1 1>
    componentLinearVelScale=<1 1 1>
    gravityOverride {z=-650}
    physicsBodyDefinitions [{
      sphereCollisionRadius=0
      boundBone {boneName=dyn_bun_3}
      boxExtents=<1 5 5>
      constraintSetup {angularTargetAxis=AxisY, angularLimitsMax=<0 0.5 0.5>, angularLimitsMin=<0 -0.5 -0.5>, angularTarget=<0 0 0>, linearAxesMax=<1 0 0>}
      localJointOffset=<2 0 0>
     }]
   }
   Output Pose(Result=Component To Local(ComponentPose=@n8)) {layerGroup=None}
  ]}
 EventGraph {links [], nodes []}
]
variables []
